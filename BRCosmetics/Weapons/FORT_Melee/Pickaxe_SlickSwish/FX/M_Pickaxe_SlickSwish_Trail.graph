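# Translucent unlit trail material for the SlickSwish pickaxe, built for Niagara ribbons and mesh particles.
# A panning trail texture, boosted by scrolling caustic detail and shaped by a linear mask, drives color and opacity.
# Dynamic parameters control offsets, variant color, noise-broken appear and subtractive dissolve.
# Scalar lerps toggle edge, appear, glider and dissolve masks, followed by view-angle fade and depth fade, which forward and mobile paths skip.
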Material /BRCosmetics/Weapons/FORT_Melee/Pickaxe_SlickSwish/FX/M_Pickaxe_SlickSwish_Trail
graph {
 properties {
  bAutomaticallySetUsageInEditor=false
  bUsedWithNiagaraMeshParticles=true
  bUsedWithNiagaraRibbons=true
  blendMode=BLEND_Translucent
  shadingModel=MSM_Unlit
  usedShadingModels=Unlit
 }
 nodes [
  MaterialFunctionCall() {materialFunction=&DebugTimeSine}
  n1 = LinearInterpolate(
   A=VectorParameter() {parameterName=Color, defaultValue=<rgba 1 0.1446 0.4842 1>}.RGB
   B=VectorParameter() {parameterName=VariantColor, defaultValue=<rgba 0.4 0.1 1 1>}.RGB
   Alpha=DynamicParameter() {defaultValue=<rgba 1 0 0 0>, paramNames [U_Offset, Dissolve, Appear, Use_VariantColor]}.Use_VariantColor
  )
  n2 = ScalarParameter() {parameterName=DetailTex_Uoffset}
  n3 = DynamicParameter() {parameterIndex=1, paramNames [DetailTex_SpeedOffset, DetailTex_UVOffset, Param3, Param4]}
  n4 = ScalarParameter() {parameterName=DetailTex_Voffset}
  n5 = Add(A=TextureCoordinate(), B=AppendVector(A=Add(A=@n2, B=@n3.DetailTex_UVOffset), B=Add(A=@n4, B=@n3.DetailTex_UVOffset)))
  n6 = VectorParameter() {parameterName=DetailTexUV, defaultValue=<rgba 1.5 1 0.1 2>}
  n7 = AppendVector(A=@n6.R, B=@n6.R)
  n8 = Multiply(A=@n6.B, B=@n3.DetailTex_SpeedOffset)
  n9 = Multiply(A=@n6.A, B=@n3.DetailTex_SpeedOffset)
  n10 = TextureSampleParameter2D(UVs=Panner(Coordinate=Multiply(A=@n5, B=@n7), Speed=AppendVector(A=@n8, B=@n9))) {parameterName=DetailTex, samplerType=SAMPLERTYPE_LinearGrayscale, texture=/Game/Global/Textures/Default/Caustics/T_Nosie_Caustics_Grayscale}
  n11 = VectorParameter() {parameterName=TexA_UVScale, defaultValue=<rgba 1.2 1 0 0>}
  n12 = Add(
   A=Multiply(A=TextureCoordinate(), B=ComponentMask(None=@n11.RGB) {g=true, r=true})
   B=AppendVector(A=DynamicParameter() {defaultValue=<rgba 1 0 0 0>, paramNames [U_Offset, Dissolve, Appear, Use_VariantColor]}.U_Offset, B=Constant())
  )
  n13 = AppendVector(A=ScalarParameter() {defaultValue=0.5, parameterName=TexA_SpeedX}, B=Constant())
  n14 = StaticComponentMaskParameter(None=TextureSampleParameter2D(UVs=Panner(Coordinate=@n12, Speed=@n13) {speedX=0.5}) {parameterName=TexA, texture=&/T_Pickaxe_SlickSwish_Trail}.RGB) {defaultR=true, parameterName=LerpTexMask}
  n15 = Multiply(A=Multiply(A=@n10.R, B=ScalarParameter() {defaultValue=2, parameterName=AddTex_Mul}) {constB=2}, B=@n14)
  n16 = Constant()
  n17 = ScalarParameter() {defaultValue=-0.5, parameterName=Linearmask_Voffset}
  n18 = ComponentMask(None=Add(A=TextureCoordinate(), B=AppendVector(A=@n16, B=@n17))) {g=true}
  n19 = Power(Base=Multiply(A=OneMinus(None=@n18)) {constB=10}) {constExponent=10}
  n20 = ScalarParameter() {defaultValue=0.3, parameterName=Lerptexture}
  n21 = ScalarParameter() {defaultValue=0.5, parameterName=LerpTex_SpeedX}
  n22 = Constant()
  n23 = TextureSampleParameter2D(UVs=Panner(Coordinate=TextureCoordinate(), Speed=AppendVector(A=@n21, B=@n22)) {speedX=0.5}) {parameterName=LerpTex, samplerType=SAMPLERTYPE_Grayscale, texture=&LowResBlurredNoise}
  n24 = Multiply(
   A=Multiply(A=Add(A=@n15, B=@n14), B=Saturate(None=@n19))
   B=LinearInterpolate(A=LinearInterpolate(A=@n20, Alpha=@n23.R) {constA=0.3}, B=Constant() {r=1}, Alpha=ScalarParameter() {parameterName=Use_LerpTex})
  )
  n25 = ParticleColor()
  MaterialFunctionCall(Emissive=Multiply(A=Multiply(A=@n1, B=@n24), B=@n25.RGB)) {materialFunction=&MF_FX_EmissiveAdjustments}
  n26 = TextureCoordinate()
  n27 = ComponentMask(None=@n26) {r=true}
  n28 = Power(Base=Multiply(A=OneMinus(None=@n27)) {constB=1.5})
  Multiply(A=Multiply(A=ComponentMask(None=@n26) {r=true}) {constB=2}, B=@n28)
  n29 = TextureCoordinate() {uTiling=1.5, vTiling=1.5}
  Power(Base=Multiply(A=TextureSample(UVs=@n29) {samplerType=SAMPLERTYPE_Grayscale, texture=/BRCosmetics/Weapons/FORT_Melee/Pickaxe_Cube_Ninja_Male/FX/T_CellNoise}.R) {constB=10}) {constExponent=10}
  n30 = DynamicParameter() {defaultValue=<rgba 1 0 0 0>, paramNames [U_Offset, Dissolve, Appear, Use_VariantColor]}
  n31 = TextureCoordinate()
  n32 = ComponentMask(None=VectorParameter() {parameterName=TexA_UVScale, defaultValue=<rgba 1.2 1 0 0>}.RGB) {g=true, r=true}
  n33 = Multiply(
   A=TextureSample(UVs=Multiply(A=@n31, B=@n32)) {texture=&/T_Pickaxe_SlickSwish_TrailMask}.R
   B=TextureSample(UVs=TextureCoordinate() {uTiling=1.5, vTiling=1.5}) {samplerType=SAMPLERTYPE_Grayscale, texture=&T_CellNoise}.R
  )
  n34 = LinearInterpolate(A=@n30.Dissolve, B=@n30.Dissolve, Alpha=Saturate(None=Power(Base=@n33) {constExponent=3}))
  n35 = Constant()
  n36 = DynamicParameter() {defaultValue=<rgba 1 0 0 0>, paramNames [U_Offset, Dissolve, Appear, Use_VariantColor]}
  n37 = ComponentMask(None=Add(A=TextureCoordinate(), B=AppendVector(A=@n35, B=@n36.Appear))) {g=true}
  n38 = Multiply(A=Power(Base=Sine(None=@n37) {period=2}))
  n39 = TextureCoordinate() {uTiling=3}
  n40 = Multiply(
   A=@n38
   B=Add(
    A=@n38
    B=TextureSample(UVs=@n39) {samplerType=SAMPLERTYPE_Grayscale, texture=&T_CellNoise}.R
   )
  )
  n41 = Multiply(A=@n25.A, B=@n24)
  n42 = ComponentMask(None=TextureCoordinate()) {r=true}
  n43 = Multiply(A=Power(Base=Sine(None=@n42) {period=2})) {constB=7}
  n44 = LinearInterpolate(A=Multiply(A=@n41, B=Saturate(None=@n43)), B=@n41, Alpha=ScalarParameter() {parameterName=Use_EdgeMask})
  n45 = LinearInterpolate(A=@n44, B=Multiply(A=@n44, B=Saturate(None=@n40)), Alpha=ScalarParameter() {parameterName=Use_Appear})
  n46 = LinearInterpolate(A=@n45, B=Multiply(A=@n45, B=@n28), Alpha=ScalarParameter() {parameterName=Use_GliderTrailMask})
  n47 = Multiply(
   A=LinearInterpolate(A=@n46, B=Subtract(A=@n46, B=@n34), Alpha=ScalarParameter() {defaultValue=1, parameterName=Use_Dissolve})
   B=MaterialFunctionCall("FadeSensitivity (optional)"=ScalarParameter() {parameterName=MeshFade}) {materialFunction=&FadeBasedOnViewAngle}.Result
  )
  n48 = DepthFade(Opacity=@n47, FadeDistance=ScalarParameter() {defaultValue=5, parameterName=DepthFade})
  ShadingPathSwitch(Default=@n48, Deferred=@n48, Forward=@n47, Mobile=@n47)
 ]
}
